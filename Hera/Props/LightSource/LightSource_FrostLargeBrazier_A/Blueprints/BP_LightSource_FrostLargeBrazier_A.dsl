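; Large frost brazier light-source prop that applies Curie fire interaction to actors inside its fire volume and supports recolorable flame effects.
; On authority with Curie active, overlap begins a fire interaction and stores its handle per actor in a map, ending it on exit.
; The construction script converts FX Color's hue offset from reference orange into hue shifts on fire, smoke, ember and core materials.
; It also tints the mesh's emissive with the point light's color.
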(event EventBeginPlay
  (|CallParentFunction))

(event OnComponentBeginOverlap(Fire_CollisionVolume)
  (params (OverlappedComponent :type Primitive%20Component%20Object%20Reference) (OtherActor :type Actor%20Object%20Reference) (OtherComp :type Primitive%20Component%20Object%20Reference) (OtherBodyIndex :type Integer) (bFromSweep :type Boolean) (SweepResult :type Hit%20Result%20Structure%20%28by%20ref%29))
  ;[1 1 1] Handle fire volume overlap. Deal fire damage if Curie is enabled.
  (Networking|SwitchHasAuthority
    (:Authority
      (if (Curie|IsCurieActive)
        (if (not (Utilities|Map|Contains (Variables|Default|GetInteractHandleMap) OtherActor))
          (bind _returnvalue (Curie|ExecuteBeginFireInteract self self OtherActor 1e+02 SweepResult))
          (Utilities|Map|Add (Variables|Default|GetInteractHandleMap) OtherActor _returnvalue))))))

(event OnComponentEndOverlap(Fire_CollisionVolume)
  (params (OverlappedComponent :type Primitive%20Component%20Object%20Reference) (OtherActor :type Actor%20Object%20Reference) (OtherComp :type Primitive%20Component%20Object%20Reference) (OtherBodyIndex :type Integer))
  (bind _interacthandlemap (Variables|Default|GetInteractHandleMap))
  (bind (_value _returnvalue) (Utilities|Map|Find _interacthandlemap OtherActor))
  ;[1 1 1] Handle fire volume end overlap. Remove fire damage if Curie is enabled.
  (Networking|SwitchHasAuthority
    (:Authority
      (if (Curie|IsCurieActive)
        (if _returnvalue
          (Curie|ExecuteEndFireInteract OtherActor _value)
          (Utilities|Map|Remove (Variables|Default|GetInteractHandleMap) OtherActor))))))

(fn ConstructionScript
  (bind _returnvalue (Math|Color|RGBtoHSV(LinearColor) "(R=1.000000,G=0.250000,B=0.000000,A=1.000000)"))
  (bind _fx_color (Variables|Default|GetFX%20Color))
  (bind _returnvalue_2 (Math|Color|RGBtoHSV(LinearColor) _fx_color))
  (bind _returnvalue_3 (- _returnvalue_2 _returnvalue))
  (bind _r (Utilities|Struct|BreakLinearColor _returnvalue_3))
  (bind _returnvalue_4 (/ _r 3.6e+02))
  (|CallParentFunction)
  (bind _returnvalue_5 (Rendering|Material|CreateDynamicMaterialInstance (Variables|Default|GetFire_Static) 0 (Rendering|Material|GetMaterial (Variables|Default|GetFire_Static) 0) "None" :!class /Script/Engine.PrimitiveComponent))
  (Rendering|Material|SetScalarParameterValue _returnvalue_5 "Hue Shift Percentage" _returnvalue_4 :!class /Script/Engine.MaterialInstanceDynamic)
  (bind _returnvalue_6 (Rendering|Material|CreateDynamicMaterialInstance (Variables|Default|GetSM_Smokestack_Sprites_01) 0 (Rendering|Material|GetMaterial (Variables|Default|GetSM_Smokestack_Sprites_01) 0) "None" :!class /Script/Engine.PrimitiveComponent))
  (Rendering|Material|SetScalarParameterValue _returnvalue_6 "Hue Shift Percentage (S)" _returnvalue_4 :!class /Script/Engine.MaterialInstanceDynamic)
  (bind _returnvalue_7 (Rendering|Material|CreateDynamicMaterialInstance (Variables|Default|GetEmbers) 0 (Rendering|Material|GetMaterial (Variables|Default|GetEmbers) 0) "None" :!class /Script/Engine.PrimitiveComponent))
  (Rendering|Material|SetScalarParameterValue _returnvalue_7 "Hue Shift Percentage (S)" _returnvalue_4 :!class /Script/Engine.MaterialInstanceDynamic)
  (bind _returnvalue_8 (Rendering|Material|CreateDynamicMaterialInstance (Variables|Default|GetCoreFire) 0 (Rendering|Material|GetMaterial (Variables|Default|GetCoreFire) 0) "None" :!class /Script/Engine.PrimitiveComponent))
  (Rendering|Material|SetScalarParameterValue _returnvalue_8 "Hue Shift Percentage (S)" _returnvalue_4 :!class /Script/Engine.MaterialInstanceDynamic)
  (bind _returnvalue_9 (Rendering|Material|CreateDynamicMaterialInstance (Variables|Debug|GetStaticMeshComponent) 0 (Rendering|Material|GetMaterial (Variables|Debug|GetStaticMeshComponent) 0) "None" :!class /Script/Engine.PrimitiveComponent))
  (Rendering|Material|SetVectorParameterValue _returnvalue_9 "SFX_Emissive_Tint(Expand)_2" (Rendering|Components|Light|GetLightColor (Variables|Default|GetPointLight)) :!class /Script/Engine.MaterialInstanceDynamic))
